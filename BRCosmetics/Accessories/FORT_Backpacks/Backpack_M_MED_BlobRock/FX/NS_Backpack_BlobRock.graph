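# Niagara effect for the BlobRock backpack cosmetic, adding a pulsing glow, a light and drifting light rays.
# GlowAndLight burst-spawns sprite, light and mesh renderers, with waveforms driving glow and light color from system color and brightness.
# LightRays continuously spawns randomly scaled meshes on a small plane that creep outward and fade in and out on a curve.
# A front-end modifier gives glow alpha separate gameplay and front-end levels.
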NiagaraSystem /BRCosmetics/Accessories/FORT_Backpacks/Backpack_M_MED_BlobRock/FX/NS_Backpack_BlobRock
dependencies {
 usedDataInterfaces [&NiagaraDataInterfaceColorCurve, &NiagaraDataInterfaceCurve, &NiagaraDataInterfaceArrayColor]
 usedDynamicInputs [&FNTimeOfDayScalar, &SelectLinearColorFromArray]
 usedModules [
  &self:SystemSpawnScript.SystemScriptSource.SystemScriptGraph.NiagaraNodeAssignment_1.SetVariables_#1
  &SystemState
  &self:SystemSpawnScript.SystemScriptSource.SystemScriptGraph.NiagaraNodeAssignment_5.SetVariables_#2
  &self:SystemSpawnScript.SystemScriptSource.SystemScriptGraph.NiagaraNodeAssignment_0.SetVariables_#3
  &self:SystemSpawnScript.SystemScriptSource.SystemScriptGraph.NiagaraNodeAssignment_2.SetVariables_#4
 ]
}
emitters {
 GlowAndLight {
  emitterSpawnScript [
   /SetVariables(Emitter.LightPos:NiagaraPosition=<0 -13.3 4.1>, Emitter.Position:NiagaraPosition=$System.SimPos) {isSetParameters=true}
   /SetVariables(Emitter.MeshPos:NiagaraPosition=<0 -16.5 2.5>) {isSetParameters=true}
  ]
  emitterUpdateScript [
   &EmitterState(
    Inactive Response=::NewEnumerator1
    Life Cycle Mode=::NewEnumerator1
    Loop Behavior=::NewEnumerator1
    Loop Duration Mode=::NewEnumerator1
    UseLoopDelay=false
   )
   &SpawnBurst_Instantaneous(Use Spawn Probability=false, UseLoopCountLimit=false)
   /SetVariables(
    Emitter.GlowColor:LinearColor=&MultiplyLinearColor(
     Linear Color=$System.Color
     Scale Factor=&MakeFloat4FromVectorAndFloat(
      "Vector (RGB)"=&VectorFromFloat(Value=$System.BrightnessMultiplier)
      "Float (Alpha)"=&Multiply_Float(
       A=&FN_FrontEndFloatModifier(Gameplay=0.09, Front End=0.07)
       B=&Waveform("Waveform [1]"=::NewEnumerator0, "[1] Frequency"=3, "[1] Amplitude Min/Max"=<0.7 1>)
      )
     )
    )
    Emitter.LightColor:LinearColor=&MultiplyLinearColor(
     Linear Color=$System.Color
     Scale Factor=&MakeVector4_FromFloat(Float=&Multiply_Float(
       A=$System.BrightnessMultiplier
       B=&Multiply_Float(A=4, B=&Waveform("Waveform [1]"=::NewEnumerator0, "[1] Frequency"=3, "[1] Amplitude Min/Max"=<0.7 1>))
      ))
    )
   ) {isSetParameters=true}
  ]
  particleSpawnScript [
   &InitializeParticle(
    Color Mode=::NewEnumerator0
    Lifetime=0
    Mesh Scale=<0.08 0.05 0.03>
    Mesh Scale Mode=::NewEnumerator4
    Position Offset:NiagaraPosition=$System.SimPos
    Sprite Size=<47.5 41.6667>
    Sprite Size Mode=::NewEnumerator4
    UsePositionOffset=true
    Write Material Random=true
   )
   &SpriteFacingAndAlignment(Sprite Facing=<0 -1 0.1>, WriteAlignment=true, WriteFacing=true)
   &CameraOffset(Camera Offset Amount=9)
   &DynamicMaterialParameters(
    Index 0 Param 1=1
    Module.Index0Float0Write=true
    Module.Index0Float1Write=false
    Module.Index0Float2Write=false
    Module.Index0Float3Write=false
   )
   /SetVariables() {isSetParameters=true}
   &InitialMeshOrientation(
    Mesh Axis To Orient=<0 1 0.055>
    Mesh Orientation Mode=::NewEnumerator2
    Orientation Vector=<0 0 1>
    Rotation=<0 0 0.25>
    Use Rotation Vector=false
   )
  ]
  rendererClasses [
   &NiagaraSpriteRendererProperties
   &NiagaraLightRendererProperties
   &NiagaraMeshRendererProperties
  ]
  renderers [
   {rendererClass=&NiagaraSpriteRendererProperties}
   {rendererClass=&NiagaraLightRendererProperties, rendererIndex=1}
   {rendererClass=&NiagaraMeshRendererProperties, rendererIndex=2}
  ]
 }
 LightRays {
  emitterUpdateScript [
   &EmitterState(Life Cycle Mode=::NewEnumerator1, Loop Duration=1, UseLoopDelay=false)
   &SpawnRate(SpawnRate=2, Use Spawn Probability=false)
   /SetVariables(Emitter.GlowColor:LinearColor=&MultiplyLinearColor(
     Linear Color=$System.Color
     Scale Factor=&MakeFloat4FromVectorAndFloat(
      "Vector (RGB)"=&VectorFromFloat(Value=$System.BrightnessMultiplier)
      "Float (Alpha)"=&Multiply_Float(
       A=&FN_FrontEndFloatModifier(Gameplay=0.09, Front End=0.07)
       B=&Waveform("Waveform [1]"=::NewEnumerator0, "[1] Frequency"=3, "[1] Amplitude Min/Max"=<0.7 1>)
      )
     )
    )) {isSetParameters=true}
  ]
  particleSpawnScript [
   &InitializeParticle(
    Lifetime=5
    Mesh Scale Max=<0.1 0.1 0.06>
    Mesh Scale Min=<0.05 0.05 0.035>
    Mesh Scale Mode=::NewEnumerator2
    Position Offset=<0 -10 4.1>
    UsePositionOffset=true
    Write Material Random=true
   )
   &CameraOffset(Camera Offset Amount=-5)
   &ShapeLocation(
    Box / Plane Mode:ENiagara_BoxPlaneMode=::NewEnumerator1
    MaskSpawnGroup=false
    Override Seed=false
    Plane Midpoint:Vector2f=<0.5 0.26>
    Plane Size:Vector2f=<12.816 8.7791>
    Rotation Axis=<0 1 0>
    Rotation Mode=::NewEnumerator1
    Shape Primitive=::NewEnumerator2
    UseOwnerScale=true
   )
   &InitialMeshOrientation(Mesh Orientation Mode=::NewEnumerator2, Orientation Vector=<0 0 1>, Use Rotation Vector=true)
   &AddVelocity(
    Velocity=&Add_Vector(
     A=<0 -1 0>
     B=&Multiply_VectorByFloat(
      Vector=&NormalizeVector(
       Vector To Normalize=&SubtractVectorFromPosition(Position=$Particles.Position, Vector=<0 -10 4.1>)
       Safe Normalize=true
       Safe Fallback Threshold=0
       Safe Fallback Vector=<0 0 0.001>
      )
      Float=0.2
     )
    )
    Velocity Speed Scale=0.01
   )
  ]
  particleUpdateScript [
   &ParticleState()
   &ScaleMeshSize(Scale Factor=&MakeVector(X=1, Y=1, Z=1))
   &SolveForcesAndVelocity(Clamp Velocity=false, Limit Acceleration=false)
   &ScaleColor(
    Scale Mode=::NewEnumerator1
    Scale RGBA=/Niagara/DynamicInputs/Multiply/Multiply_Vector4_ByFloat(
     Vector4:Vector4f=&MakeVector4_FromLinearColor(LinearColor=$Emitter.GlowColor)
     Float:NiagaraFloat=&FloatFromCurve(FloatCurve=DataInterface() {propertyValues {Curve {keys [<at 0 0 RCIM_Cubic>,<at 0.4 1 RCIM_Cubic>,<at 1 0 RCIM_Cubic>]}}}, Scale Curve=2.2)
    )
    ScaleRGBA=true
   )
   &DynamicMaterialParameters(
    Index 0 Param 1=&RandomRangeFloat(Minimum=0.05, Maximum=-0.05)
    Module.Index0Float0Write=true
    Module.Index0Float1Write=true
    Module.Index0Float2Write=true
    Module.Index0Float3Write=true
   )
  ]
  rendererClasses [&NiagaraMeshRendererProperties]
  renderers [{rendererClass=&NiagaraMeshRendererProperties, rendererIndex=0}]
 }
}
summary {
 systemName=NS_Backpack_BlobRock
 emitters [
  {
   emitterName=GlowAndLight
   rendererClasses [
    &NiagaraSpriteRendererProperties
    &NiagaraLightRendererProperties
    &NiagaraMeshRendererProperties
   ]
  }
  {emitterName=LightRays, rendererClasses [&NiagaraMeshRendererProperties]}
 ]
 userVariables [
  User.bIsFrontEnd {type=NiagaraBool}
  User.bIsFrontEndPreview {type=NiagaraBool}
  User.LightID {type=NiagaraInt32}
 ]
}
