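# Noise material function for the BisonDrain TV gadget prop, built by chaining the MF_iHashForSumming hash function.
# The Vector2 input IN_V, previewed with texture coordinates, enters the first hash through a named reroute with a constant index of 1.
# Each later call takes the previous Result as both IN_V and IN_i.
# Only the second of the eight calls feeds the output; the remaining six form an unused tail.
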MaterialFunction /BRCosmetics/Animation/Props/Gadgets_BisonDrain_TV/Materials/MF_NoiseBisonDrain
graph {nodes [
  n1 = MaterialFunctionCall!(IN_V=NamedRerouteUsage(), IN_i=Constant() {r=1}) {materialFunction=&/MF_iHashForSumming}
  n2 = MaterialFunctionCall(IN_V=@n1.Result, IN_i=@n1.Result) {materialFunction=&MF_iHashForSumming}
  FunctionOutput(None=@n2.Result)
  n3 = MaterialFunctionCall(IN_V=@n2.Result, IN_i=@n2.Result) {materialFunction=&MF_iHashForSumming}
  n4 = MaterialFunctionCall(IN_V=@n3.Result, IN_i=@n3.Result) {materialFunction=&MF_iHashForSumming}
  n5 = MaterialFunctionCall(IN_V=@n4.Result, IN_i=@n4.Result) {materialFunction=&MF_iHashForSumming}
  n6 = MaterialFunctionCall(IN_V=@n5.Result, IN_i=@n5.Result) {materialFunction=&MF_iHashForSumming}
  n7 = MaterialFunctionCall(IN_V=@n6.Result, IN_i=@n6.Result) {materialFunction=&MF_iHashForSumming}
  MaterialFunctionCall(IN_V=@n7.Result, IN_i=@n7.Result) {materialFunction=&MF_iHashForSumming}
  NamedRerouteDeclaration(None=FunctionInput(Preview=TextureCoordinate()) {inputName=IN_V, inputType=FunctionInput_Vector2}) {name=IN_V, nodeColor=<rgba 1 0.0703 0 1>}
 ]}
