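# Material function for the BlobRock backpack's screen, overriding BaseColor and EmissiveColor on incoming material attributes to show an animated flipbook with a VHS-style look.
# Scaled and offset UV1 coordinates feed noise-driven tape wave, crease and switching distortions, reusing BisonDrain TV noise and hash functions.
# Hashed scanline pulses, a bloom-sum flipbook function and a vignette texture shape the glow.
# Time and timestamp logic, with a debug switch, drive flipbook phase.
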MaterialFunction /BRCosmetics/Accessories/FORT_Backpacks/Backpack_M_MED_BlobRock/Materials/MF_Backpack_BlobRock_Screen
graph {nodes [
  n1 = GetMaterialAttributes(None=NamedRerouteUsage()) {attributeGetTypes [#1, #2]}
  FunctionOutput(None=SetMaterialAttributes!(MaterialAttributes=@n1.MaterialAttributes, EmissiveColor=NamedRerouteUsage(), BaseColor=NamedRerouteUsage()) {attributeSetTypes [#1, #2]})
  n2 = ComponentMask(None=TextureCoordinate() {coordinateIndex=1}) {g=true}
  Multiply(A=@n2, B=ScalarParameter() {defaultValue=1, parameterName=VignetteYScale})
  n3 = Multiply(
   A=TextureCoordinate() {coordinateIndex=1}
   B=AppendVector(A=ScalarParameter() {defaultValue=1, parameterName=UVScaleX}, B=ScalarParameter() {defaultValue=1, parameterName=UVScaleY})
  )
  n4 = AppendVector(A=ScalarParameter() {parameterName=UVPositionX}, B=ScalarParameter() {parameterName=UVPositionY})
  NamedRerouteDeclaration(None=NamedRerouteDeclaration(None=Add(A=@n3, B=@n4)) {name=UVs, nodeColor=<rgba 0.0547 0 1 1>}) {name=UVn, nodeColor=<rgba 0.2656 1 0 1>}
  n5 = Subtract(A=ComponentMask(None=NamedRerouteUsage()) {r=true}, B=Multiply(A=NamedRerouteUsage(), B=NamedRerouteUsage()))
  n6 = NamedRerouteUsage()
  NamedRerouteDeclaration(None=AppendVector(A=NamedRerouteDeclaration(None=@n5) {name=UVnx_tapecrease, nodeColor=<rgba 0.8281 1 0 1>}, B=ComponentMask(None=@n6) {g=true})) {name=UVn_tapecrease, nodeColor=<rgba 0.2891 0 1 1>}
  n7 = ConstantBiasScale(None=Sine(None=Time() {bOverride_Period=true, period=1}))
  n8 = NamedRerouteUsage()
  NamedRerouteDeclaration(None=SmoothStep(Min=Multiply(A=@n7) {constB=0.02}, Value=ComponentMask(None=@n8) {g=true}) {constMax=0, constMin=0.02}) {name=SwitchingNoise_phase, nodeColor=<rgba 1 0 0.9141 1>}
  n9 = NamedRerouteUsage()
  n10 = NamedRerouteUsage()
  NamedRerouteDeclaration(None=Add(
    A=MaterialFunctionCall(Float2=@n9) {materialFunction=&BreakOutFloat2Components}.G
    B=Multiply(A=@n10)
   )) {name=UVny_switchingnoise, nodeColor=<rgba 0 1 0.7656 1>}
  n11 = AppendVector(A=Multiply(A=NamedRerouteUsage()) {constB=100}, B=Multiply(A=Time()))
  n12 = Multiply(A=Subtract(A=MaterialFunctionCall!(IN_V=@n11) {materialFunction=/BRCosmetics/Animation/Props/Gadgets_BisonDrain_TV/Materials/MF_NoiseBisonDrain}.Result) {constB=0.5}) {constB=0.2}
  n13 = NamedRerouteUsage()
  n14 = NamedRerouteUsage()
  NamedRerouteDeclaration(None=Add(A=Multiply(A=@n12, B=@n13), B=ComponentMask(None=@n14) {r=true})) {name=UVnx_switchingnoise, nodeColor=<rgba 1 0.4453 0 1>}
  NamedRerouteDeclaration(None=AppendVector(A=NamedRerouteUsage(), B=NamedRerouteUsage())) {name=UVn_switchingnoise, nodeColor=<rgba 0.125 0 1 1>}
  n15 = ComponentMask(None=NamedRerouteUsage()) {g=true}
  n16 = Multiply(A=Add(A=Time() {bOverride_Period=true, period=60}) {constB=60}) {constB=0.5}
  n17 = MaterialFunctionCall!(IN_V=AppendVector(A=Constant(), B=Add(A=@n15, B=@n16)), IN_R=NamedRerouteUsage()) {materialFunction=/BRCosmetics/Animation/Props/Gadgets_BisonDrain_TV/Materials/MF_iHash}
  n18 = ScalarParameter() {defaultValue=0.5, group=BisonDrainScreen, parameterName=ScanLineMultiplier}
  n19 = Saturate(None=SmoothStep(
    Min=ScalarParameter() {defaultValue=0.1, parameterName=Min}
    Max=ScalarParameter() {defaultValue=0.33, parameterName=Max}
    Value=Multiply(A=@n17.Result, B=@n18) {constB=0.3}
   ) {constMax=0.532, constMin=0.104})
  NamedRerouteDeclaration(None=Multiply(A=NamedRerouteUsage(), B=Add(A=@n19) {constB=-0.4})) {name=PulsedBloomedColor, nodeColor=<rgba 0 0.0156 1 1>}
  NamedRerouteDeclaration(None=@n1.BaseColor) {name=In_BaseColor, nodeColor=<rgba 1 0.9141 0 1>}
  n20 = ScalarParameter() {defaultValue=4, parameterName=FlipbookRowSize}
  n21 = ScalarParameter() {defaultValue=3, parameterName=FlipbookColumnSize}
  n22 = TextureObjectParameter() {
   constCoordinate=1
   parameterName=BlobRockFlipbookTexture
   texture=/BRCosmetics/Accessories/FORT_Backpacks/Backpack_M_MED_BlobRock/Textures/T_BlobRock_Idle
  }
  n23 = RecordTextureStreamingInfo!(Tex=@n22, UVs=NamedRerouteUsage())
  n24 = NamedRerouteUsage()
  n25 = ScalarParameter() {defaultValue=0.02, group=BisonDrainScreen, parameterName=PulseFlipbookUVMultiplier}
  n26 = Multiply(A=NamedRerouteUsage(), B=Add(B=Multiply(A=@n24, B=@n25) {constB=12}) {constA=1})
  NamedRerouteDeclaration(None=MaterialFunctionCall(
    "Animation  Phase (0-1)"=NamedRerouteUsage()
    Number of Rows=NamedRerouteDeclaration(None=@n20) {name=FlipbookRowSize, nodeColor=<rgba 0.0781 0 1 1>}
    Number of Columns=NamedRerouteDeclaration(None=@n21) {name=FlipbookColumnSize, nodeColor=<rgba 0.2422 1 0 1>}
    Texture=NamedRerouteDeclaration(None=@n23.Tex) {name=BlobRockFlipbookTexture, nodeColor=<rgba 0.5938 0 1 1>}
    UVs=NamedRerouteDeclaration(None=@n26) {name=UVForFlipbooks, nodeColor=<rgba 0 1 0.6016 1>}
   ) {materialFunction=&FlipBook}.Result) {name=BlobRockFlipbookResult, nodeColor=<rgba 0.4297 1 0 1>}
  n27 = Divide(A=Time(), B=ScalarParameter() {defaultValue=3, group=FNCSShield_Screen, parameterName=DebugTransitionLength, sortPriority=30})
  n28 = MaterialFunctionCall(
   TimestampA=ScalarParameter() {defaultValue=1, parameterName=Timestamp, sortPriority=30}
   TransitionLength_In=ScalarParameter() {defaultValue=0.1, parameterName=TransitionLength, sortPriority=30}
  ) {materialFunction=&MF_Timestamp_InOut_01}
  NamedRerouteDeclaration(None=StaticSwitchParameter(True=Frac(None=@n27), False=VertexInterpolator(VS=@n28.Linear).PS) {group=FNCSShield_Screen_Graphics, parameterName=DebugModeOn, sortPriority=0}) {name=TimeStamp, nodeColor=<rgba 0.125 1 0 1>}
  NamedRerouteDeclaration(None=@n1.EmissiveColor) {name=In_EmissiveColor, nodeColor=<rgba 1 0 0.0469 1>}
  NamedRerouteDeclaration(None=FunctionInput() {inputName=In_MaterialAttributes, inputType=FunctionInput_MaterialAttributes}) {name=In_MaterialAttributes, nodeColor=<rgba 0 1 0.4375 1>}
  n29 = Time()
  n30 = NamedRerouteDeclaration(None=ScalarParameter() {defaultValue=0.1111, group=BisonDrainScreen, parameterName=FlipbookForTVTimeMultiplier}) {name=FlipbookTimeMultiplier, nodeColor=<rgba 1 0.8672 0 1>}
  NamedRerouteDeclaration(None=Frac(None=Multiply(A=@n29, B=@n30) {constB=0.5})) {name=NormalizedTime, nodeColor=<rgba 1 0 0.5859 1>}
  NamedRerouteDeclaration(None=Multiply(A=NamedRerouteUsage()) {constB=0.1}) {name=Out_BaseColor, nodeColor=<rgba 0 0.9297 1 1>}
  n31 = NamedRerouteUsage()
  n32 = ComponentMask(None=NamedRerouteUsage()) {b=true}
  n33 = Multiply(A=NamedRerouteUsage(), B=Add(A=Multiply(A=@n32, B=@n32)) {constB=0.05})
  n34 = ComponentMask(None=NamedRerouteUsage()) {b=true}
  NamedRerouteDeclaration(None=Multiply(A=LinearInterpolate(A=@n31, B=@n33, Alpha=@n34), B=NamedRerouteUsage())) {name=Out_Emissive, nodeColor=<rgba 0.8516 1 0 1>}
  n35 = Constant() {r=0.99}
  n36 = If(A=NamedRerouteUsage(), "A > B"=@n35, "A == B"=@n35, "A < B"=NamedRerouteUsage()) {constB=1}
  NamedRerouteDeclaration(None=If(A=NamedRerouteUsage(), "A > B"=@n36, "A == B"=@n36, "A < B"=NamedRerouteUsage()) {constB=1}) {name=FlipbookAnimationPhase, nodeColor=<rgba 1 0.7734 0 1>}
  n37 = NamedRerouteUsage()
  n38 = NamedRerouteUsage()
  NamedRerouteDeclaration(None=If(A=NamedRerouteUsage(), "A > B"=@n37, "A == B"=@n37, "A < B"=Multiply(A=@n38) {constB=2}) {constB=1}) {name=PlaySequence, nodeColor=<rgba 1 0 0.8672 1>}
  n39 = Multiply(A=NamedRerouteUsage(), B=OneMinus(None=NamedRerouteUsage()))
  n40 = LinearInterpolate(
   A=@n39
   B=MaterialFunctionCall(XYZ=@n39) {materialFunction=&Swizzle}.YXZ
   Alpha=NamedRerouteUsage()
  )
  NamedRerouteDeclaration(None=MaterialFunctionCall!(
    IN_Color=NamedRerouteDeclaration(None=@n40) {name=Color, nodeColor=<rgba 0 0.2734 1 1>}
    IN_UV=NamedRerouteUsage()
    IN_x=Constant() {r=-2}
    IN_Flipbook=NamedRerouteUsage()
    IN_RowSize=NamedRerouteUsage()
    IN_ColumnSize=NamedRerouteUsage()
    IN_FlipbookTimeMultiplier=NamedRerouteUsage()
   ) {materialFunction=&/MF_BloomFlipbookSumBlobRock}.BloomedColor) {name=BloomedColorFinal, nodeColor=<rgba 0.5937 1 0 1>}
  NamedRerouteDeclaration(None=FunctionInput() {inputName=In_TechArtMask, sortPriority=1}) {name=TechArtMask, nodeColor=<rgba 1 0.5859 0 1>}
  n41 = TextureSampleParameter2D() {
   constCoordinate=1
   parameterName=VignetteMask
   samplerType=SAMPLERTYPE_LinearColor
   texture=/BRCosmetics/Accessories/FORT_Backpacks/Backpack_M_MED_BlobRock/Textures/T_BlobRock_Vignette
  }
  NamedRerouteDeclaration(None=@n41.RGB) {name=Vignette, nodeColor=<rgba 0 0.5781 1 1>}
  NamedRerouteDeclaration(None=ScalarParameter() {parameterName=BlobRockFlipbook}) {name=BlobRockFlipbook, nodeColor=<rgba 0.8516 0 1 1>}
  MaterialExpressionNamedRerouteUsage_1 = NamedRerouteUsage()
  n42 = Convert(None=@MaterialExpressionNamedRerouteUsage_1) {
   nodeName=Break Float2
   convertInputs [{
     type=Vector2
     expressionInput {
      expression=&self:MaterialExpressionNamedRerouteUsage_1
      inputName=None
      mask=0
      maskA=0
      maskB=0
      maskG=0
      maskR=0
      outputIndex=0
     }
    }]
   convertOutputs [{type=Scalar}, {type=Scalar}]
  }
  n43 = Time()
  n44 = ScalarParameter() {defaultValue=10, group=BisonDrainScreen, parameterName=TimeMultiplier}
  n45 = MaterialFunctionCall(IN_V=AppendVector(A=Multiply(A=@n42) {constB=100}, B=Multiply(A=@n43, B=@n44))) {materialFunction=&MF_NoiseBisonDrain}
  MaterialExpressionNamedRerouteUsage_0 = NamedRerouteUsage()
  n46 = Convert(None=@MaterialExpressionNamedRerouteUsage_0) {
   nodeName=Break Float2
   convertInputs [{
     type=Vector2
     expressionInput {
      expression=&self:MaterialExpressionNamedRerouteUsage_0
      inputName=None
      mask=0
      maskA=0
      maskB=0
      maskG=0
      maskR=0
      outputIndex=0
     }
    }]
   convertOutputs [{type=Scalar}, {type=Scalar}]
  }
  n47 = AppendVector(A=@n46, B=Multiply(A=Time(), B=ScalarParameter() {defaultValue=10, group=BisonDrainScreen, parameterName=TimeMultiplier}))
  n48 = Multiply(A=Subtract(A=MaterialFunctionCall(IN_V=@n47) {materialFunction=&MF_NoiseBisonDrain}.Result) {constB=0.5}) {constB=0.005}
  n49 = Add(
   A=NamedRerouteDeclaration(None=Add(A=@n46, B=@n48)) {name=UVnx_updated, nodeColor=<rgba 1 0 0.5859 1>}
   B=Multiply(A=Subtract(A=@n45.Result) {constB=0.5}) {constB=0.01}
  )
  n50 = NamedRerouteUsage()
  NamedRerouteDeclaration(None=AppendVector(A=NamedRerouteDeclaration(None=@n49) {name=UVnx_update2, nodeColor=<rgba 0.5316 1 0.0765 1>}, B=ComponentMask(None=@n50) {g=true})) {name=UVn_tapewave, nodeColor=<rgba 0 1 0.1328 1>}
  n51 = NamedRerouteUsage()
  n52 = Time()
  n53 = Subtract(
   A=Multiply(A=ComponentMask(None=@n51) {g=true}) {constB=2}
   B=Multiply(A=MaterialFunctionCall(Multiplier=@n52) {materialFunction=&Pi}.Result) {constB=0.4}
  )
  n54 = Multiply(A=Time())
  n55 = Multiply(
   A=Subtract(A=Sine(None=@n53)) {constB=0.92}
   B=MaterialFunctionCall(IN_V=AppendVector(A=@n54, B=@n54)) {materialFunction=&MF_NoiseBisonDrain}.Result
  )
  NamedRerouteDeclaration(None=Multiply(A=Saturate(None=@n55)) {constB=10}) {name=TapeCrease_Phase, nodeColor=<rgba 1 0 0.1875 1>}
  n56 = ComponentMask(None=NamedRerouteUsage()) {g=true}
  n57 = Time()
  n58 = MaterialFunctionCall(IN_V=AppendVector(A=Multiply(A=@n56) {constB=100}, B=Multiply(A=@n57) {constB=10})) {materialFunction=&MF_NoiseBisonDrain}
  NamedRerouteDeclaration(None=Max(A=Subtract(A=@n58.Result) {constB=0.5}, B=ScalarParameter() {defaultValue=-0.03, group=BisonDrainScreen, parameterName=WarpAmountinCreases}) {constB=-0.03}) {name=TapeCrease_Noise, nodeColor=<rgba 0 0.5078 1 1>}
  ScalarParameter() {parameterName=HoldLastFrameTime}
  n59 = NamedRerouteDeclaration(None=Multiply(A=ScalarParameter() {defaultValue=1, parameterName=VignetteXScale}, B=@n2)) {name=XScaleVignette, nodeColor=<rgba 0 0.25 1 1>}
  n60 = ScalarParameter() {defaultValue=1, parameterName=VignetteYScale}
  n61 = ScalarParameter() {defaultValue=1, parameterName=VignetteYScale}
  SphereMask(A=Multiply(A=TextureCoordinate() {coordinateIndex=1}, B=AppendVector(A=@n59, B=@n60)), B=AppendVector(B=Divide(A=@n61))) {attenuationRadius=0.5, hardnessPercent=12}
 ]}
