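# Animation Blueprint for the BlobRock backpack that keeps it fitted to the character's pose and swings its exercise bands.
# Additive Modify Bone nodes offset pack and band bones using gender-based alphas and event-graph values.
# Those values derive from spine bend, plus left upper-arm yaw and an eased, head-tilt-scaled weight for melee.
# Five chained Anim Dynamics nodes then simulate band bones with springs and fixed gravity, relative to the band base.
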AnimBlueprint /BRCosmetics/Accessories/FORT_Backpacks/Backpack_M_MED_BlobRock/Meshes/M_MED_BlobRock_Pack_AnimBP
target_skeleton=&/M_MED_BlobRock_Pack_Skeleton
graphs [
 AnimGraph {nodes [
   n1 = "Transform (Modify) Bone - Bone: deform_exercise_band_r"(ComponentPose=Local To Component(LocalPose=Input Pose()), Alpha=1, Translation="-0.5,0,-1", Rotation="0,0,0", Scale="1,1,1") {
    translationMode=BMM_Additive
    translationSpace=BCS_BoneSpace
    alphaBoolBlend {blendOption=Linear}
    boneToModify {boneName=deform_exercise_band_r}
    translation {x=-0.5, z=-1}
   }
   n2 = "Transform (Modify) Bone - Bone: deform_exercise_band_l"(ComponentPose=@n1, Alpha=1, Translation="-0.5,0,1", Rotation="0,0,0", Scale="1,1,1") {
    translationMode=BMM_Additive
    translationSpace=BCS_BoneSpace
    alphaBoolBlend {blendOption=Linear}
    boneToModify {boneName=deform_exercise_band_l}
    translation {x=-0.5, z=1}
   }
   n3 = "Transform (Modify) Bone - Bone: pack"(ComponentPose=@n2, Alpha=Get Alpha F() -> (Alpha F:double), Translation="0,1,-2", Rotation="0,0,0") {
    translationMode=BMM_Additive
    translationSpace=BCS_BoneSpace
    alphaBoolBlend {blendOption=Linear}
    boneToModify {boneName=pack}
    translation {y=1, z=-2}
   }
   n4 = "Transform (Modify) Bone - Bone: pack"(ComponentPose=@n3, Alpha=Get Alpha M() -> (Alpha M:double), Translation="0,0.5,-2", Rotation="0,0,0") {
    translationMode=BMM_Additive
    translationSpace=BCS_BoneSpace
    alphaBoolBlend {blendOption=Linear}
    boneToModify {boneName=pack}
    translation {y=0.5, z=-2}
   }
   n5 = "Transform (Modify) Bone - Bone: base_exercise_bands"(ComponentPose=@n4, Alpha=1, Translation="0,0,0", Rotation=Get Backpack Rotate Crouch() -> (Backpack Rotate Crouch:Rotator), Scale="1,1,1") {
    rotationMode=BMM_Additive
    rotationSpace=BCS_BoneSpace
    translationMode=BMM_Additive
    translationSpace=BCS_BoneSpace
    alphaBoolBlend {blendOption=Linear}
    boneToModify {boneName=base_exercise_bands}
   }
   n6 = "Transform (Modify) Bone - Bone: pack"(
    ComponentPose=@n5
    Translation=Get Backpack Translation() -> (Backpack Translation:Vector)
    Rotation=Get Backpack Rotate() -> (Backpack Rotate:Rotator)
   ) {
    rotationMode=BMM_Additive
    rotationSpace=BCS_BoneSpace
    translationMode=BMM_Additive
    translationSpace=BCS_BoneSpace
    alphaBoolBlend {blendOption=Linear}
    boneToModify {boneName=pack}
   }
   n7 = "Transform (Modify) Bone - Bone: pack"(
    ComponentPose=@n6
    Alpha=Get isMelee() -> (isMelee:double)
    Translation="0,-1.5,0"
    Rotation=Get Backpack Rotate Pickaxe() -> (Backpack Rotate Pickaxe:Rotator)
   ) {
    rotationMode=BMM_Additive
    rotationSpace=BCS_BoneSpace
    translationMode=BMM_Additive
    translationSpace=BCS_BoneSpace
    alphaBoolBlend {blendOption=Linear}
    alphaScaleBiasClamp {interpSpeedDecreasing=5, interpSpeedIncreasing=5}
    boneToModify {boneName=pack}
    rotation {roll=20}
    translation {y=-1.5}
   }
   n8 = Anim Dynamics - Bone: dyn_exercise_band_01_r!(ComponentPose=@n7, Alpha=1, GravityOverride:Vector="0,0,-980", ExternalForce="0,0,0") -> (Pose) {
    angularDampingOverride=1
    angularSpringConstant=500
    bAngularSpring=true
    bOverrideAngularDamping=true
    bUseGravityOverride=true
    simulationSpace=BoneRelative
    alphaBoolBlend {blendOption=Linear}
    boundBone {boneName=dyn_exercise_band_01_r}
    componentAppliedLinearAccClamp=<2000 2000 2000>
    componentLinearAccScale=<0.4 0.4 0.4>
    componentLinearVelScale=<0.4 0.4 0.4>
    gravityOverride {z=-980}
    physicsBodyDefinitions [{
      boundBone {boneName=dyn_exercise_band_01_r}
      boxExtents=<10 10 10>
      constraintSetup {angularLimitsMax=<5 40 5>, angularLimitsMin=<-5 -5 -45>, angularTarget=<1 0 0>}
      localJointOffset=<1.7 0 0>
     }]
    relativeSpaceBone {boneName=base_exercise_bands}
   }
   n9 = Anim Dynamics - Bone: dyn_exercise_band_02_r!(ComponentPose=@n8, Alpha=1, GravityOverride:Vector="0,0,-980", ExternalForce="0,0,0") -> (Pose) {
    angularDampingOverride=1
    angularSpringConstant=500
    bAngularSpring=true
    bOverrideAngularDamping=true
    bUseGravityOverride=true
    simulationSpace=BoneRelative
    alphaBoolBlend {blendOption=Linear}
    boundBone {boneName=dyn_exercise_band_02_r}
    componentAppliedLinearAccClamp=<2000 2000 2000>
    componentLinearAccScale=<0.4 0.4 0.4>
    componentLinearVelScale=<0.4 0.4 0.4>
    gravityOverride {z=-980}
    physicsBodyDefinitions [{
      boundBone {boneName=dyn_exercise_band_02_r}
      boxExtents=<10 10 10>
      constraintSetup {angularLimitsMax=<10 30 5>, angularLimitsMin=<-10 -5 -40>, angularTarget=<1 0 0>}
      localJointOffset=<1.7 0 0>
     }]
    relativeSpaceBone {boneName=base_exercise_bands}
   }
   n10 = Anim Dynamics - Bone: dyn_exercise_band_01_l!(ComponentPose=@n9, Alpha=1, GravityOverride:Vector="0,0,-980", ExternalForce="0,0,0") -> (Pose) {
    angularDampingOverride=1
    angularSpringConstant=500
    bAngularSpring=true
    bOverrideAngularDamping=true
    bUseGravityOverride=true
    simulationSpace=BoneRelative
    alphaBoolBlend {blendOption=Linear}
    boundBone {boneName=dyn_exercise_band_01_l}
    componentAppliedLinearAccClamp=<2000 2000 2000>
    componentLinearAccScale=<0.4 0.4 0.4>
    componentLinearVelScale=<0.4 0.4 0.4>
    gravityOverride {z=-980}
    physicsBodyDefinitions [{
      boundBone {boneName=dyn_exercise_band_01_l}
      boxExtents=<10 10 10>
      constraintSetup {angularLimitsMax=<5 5 10>, angularLimitsMin=<-5 -30 -45>, angularTarget=<1 0 0>}
      localJointOffset=<1.7 0 0>
     }]
    relativeSpaceBone {boneName=base_exercise_bands}
   }
   n11 = Anim Dynamics - Bone: dyn_exercise_band_02_l!(ComponentPose=@n10, Alpha=1, GravityOverride:Vector="0,0,-980", ExternalForce="0,0,0") -> (Pose) {
    angularDampingOverride=1
    angularSpringConstant=500
    bAngularSpring=true
    bOverrideAngularDamping=true
    bUseGravityOverride=true
    simulationSpace=BoneRelative
    alphaBoolBlend {blendOption=Linear}
    boundBone {boneName=dyn_exercise_band_02_l}
    componentAppliedLinearAccClamp=<2000 2000 2000>
    componentLinearAccScale=<0.4 0.4 0.4>
    componentLinearVelScale=<0.4 0.4 0.4>
    gravityOverride {z=-980}
    physicsBodyDefinitions [{
      boundBone {boneName=dyn_exercise_band_02_l}
      boxExtents=<10 10 10>
      constraintSetup {angularLimitsMax=<10 5 10>, angularLimitsMin=<-10 -30 -40>, angularTarget=<1 0 0>}
      localJointOffset=<1.7 0 0>
     }]
    relativeSpaceBone {boneName=base_exercise_bands}
   }
   n12 = Anim Dynamics - Bone: dyn_exercise_band_03_l!(ComponentPose=@n11, Alpha=1, GravityOverride:Vector="0,0,-980", ExternalForce="0,0,0") -> (Pose) {
    angularDampingOverride=1
    angularSpringConstant=500
    bAngularSpring=true
    bOverrideAngularDamping=true
    bUseGravityOverride=true
    simulationSpace=BoneRelative
    alphaBoolBlend {blendOption=Linear}
    boundBone {boneName=dyn_exercise_band_03_l}
    componentAppliedLinearAccClamp=<2000 2000 2000>
    componentLinearAccScale=<0.4 0.4 0.4>
    componentLinearVelScale=<0.4 0.4 0.4>
    gravityOverride {z=-980}
    physicsBodyDefinitions [{
      boundBone {boneName=dyn_exercise_band_03_l}
      boxExtents=<10 10 10>
      constraintSetup {angularLimitsMax=<10 10 10>, angularLimitsMin=<-10 -40 -60>, angularTarget=<1 0 0>}
      localJointOffset=<1.7 0 0>
     }]
    relativeSpaceBone {boneName=base_exercise_bands}
    simSpaceSettings {maxAngularAcceleration=3000, maxAngularVelocity=3000, simSpaceAngularAlpha=0.5}
   }
   Output Pose(Result=Component To Local(ComponentPose=@n12)) {layerGroup=None}
  ]}
 EventGraph {nodes [
   Event BlueprintUpdateAnimation()
   > n1 = Sequence!(execute) -> (then_0, then_1, then_2:Exec)
   > n2 = Gravity Override State Function DEPRECATED(CharacterPartAnimInstance=Self-Reference(), GameTimeInSeconds=GetGameTimeInSeconds(), GravityOverrideParams=Get Gravity Override Params())
   > Set Gravity Override(Gravity Override=@n2)
   n3 = Branch(execute=@n1.then_2, Condition=Get bIsSkyDiving())
   n4 = GetOwnerSkeletalMeshComponent()
   n5 = BreakRotator(InRot=BreakTransform(InTransform=GetDeltaTransformFromRefPose(self=@n4, BoneName=spine_05, BaseName=spine_01)).Rotation)
   n6 = MapRangeClamped(Value=@n5.Yaw, InRangeA=-10, InRangeB=-50, OutRangeB=40)
   > Set Backpack Rotate(Backpack Rotate:Rotator=MakeRotator(Roll=@n6, Pitch=float / float(A=@n6, B=5))) -> (Output_Get:Rotator)
   > Set Backpack Translation(Backpack Translation:Vector=MakeVector(Y=MapRangeClamped(Value=@n5.Yaw, InRangeA=-15, InRangeB=-50, OutRangeB=-10))) -> (Output_Get:Vector)
   n7 = Branch(execute=@n3.else, Condition=Get Char F() -> (Char F:bool))
   > Set Alpha F(Alpha F:double=1) -> (Output_Get:double)
   > Set Alpha M(Alpha M:double=0) -> (Output_Get:double)
   n8 = Branch(execute=@n7.else, Condition=Get bIsSliding())
   Set Alpha F(execute=@n8.else, Alpha F=0)
   > Set Alpha M(Alpha M=1)
   n9 = GetDeltaTransformFromRefPose(self=GetOwnerSkeletalMeshComponent(), BoneName=spine_05, BaseName=spine_01)
   n10 = MapRangeClamped(Value=BreakRotator(InRot=BreakTransform(InTransform=@n9).Rotation).Yaw, InRangeA=10, InRangeB=50, OutRangeB=30)
   > Set Backpack Rotate Crouch(Backpack Rotate Crouch:Rotator=MakeRotator(Roll=@n10, Pitch=float / float(A=@n10, B=5))) -> (Output_Get:Rotator)
   n11 = GetOwnerSkeletalMeshComponent()
   n12 = Set head down rot(
    execute=@n1.then_1
    head down rot:double=GetAlphaFromReferenceTransformDelta(SkeletalMeshComponent=@n11, BoneName=spine_01, BaseName=head, InputRange=(X=0,Y=-45), TwistAxis="1,0,0")
   ) -> (Output_Get:double)
   n13 = "To Float (Boolean)"(InBool=Get bIsPlayingMeleeAnim())
   n14 = float * float(B=0.2) {is=title K2Node_PromotableOperator}
   n15 = Get isMelee()
   > Set isMelee(isMelee:double=float * float(A=FInterpTo(Current=@n15, Target=@n13, DeltaTime=@n14, InterpSpeed=5), B=@n12)) -> (Output_Get:double)
   n16 = BreakRotator(InRot=BreakTransform(InTransform=GetDeltaTransformFromRefPose(self=@n11, BoneName=upperarm_l, BaseName=clavicle_l)).Rotation)
   > Set Backpack Rotate Pickaxe(Backpack Rotate Pickaxe:Rotator=MakeRotator(Roll=MapRangeClamped(Value=@n16.Yaw, InRangeA=0, InRangeB=90, OutRangeB=30))) -> (Output_Get:Rotator)
   n17 = Event OnCharacterAndPickaxeCustomizationLoaded!() -> (
    OutputDelegate
    then
    CustomBaseMesh:Skeletal Mesh Component Object Reference
    CustomParts:Array of Custom Character Part Object References
    CustomMeshParts:Array of Skeletal Mesh Component Budgeted Object References
   ) {is=title K2Node_Event}
   n18 = Get Array Item with EFort Custom Part Type(Array:Array of Custom Character Part Object References=@n17.CustomParts) -> (Output:Custom Character Part Object Reference)
   > Set Char M(Char M:bool="Equal (Enum)"(
     A:EFortCustomGender Enum=Get GenderPermitted!(self:Custom Character Part Object Reference=@n18) -> (GenderPermitted:EFortCustomGender Enum=Invalid)
     B:EFortCustomGender Enum=Male
    )) -> (Output_Get:bool)
   > Set Char F(Char F:bool="Equal (Enum)"(A=Get GenderPermitted(self=@n18) -> (GenderPermitted=Invalid), B=Female)) -> (Output_Get:bool)
   TryGetPawnOwner()
  ]}
]
variables [
 Alpha F {type=number}
 Alpha M {type=number}
 Backpack Rotate {type=Rotator}
 Backpack Rotate Crouch {type=Rotator}
 Backpack Rotate Pickaxe {type=Rotator}
 Backpack Translation {type=Vector}
 Char F {type=boolean}
 Char M {type=boolean}
 Gravity Override {type=Vector}
 Gravity Override Params {type=GravityOverrideParamsStruct}
 head down rot {type=number}
 isMelee {type=number}
]
